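# Parent material for the HoppyBun biped cosmetic companion, layered between the shared character PreFX and PostFX functions.
# A masked subsurface blend, excluded from custom color mask areas, uses the Subsurface shading model only at Epic quality, with optional Fresnel-driven opacity.
# After the companion custom color function, a procedural brow mask from tech art mask green, UVs and pre-skinned position tints secondary color.
# The companion flipbook function runs before PostFX.
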Material /CosmeticCompanions/Assets/Biped/HoppyBun/Materials/M_HoppyBun_Parent
graph {
 properties {
  bAutomaticallySetUsageInEditor=false
  bUseMaterialAttributes=true
  bUsedWithMeshDeformer=true
  bUsedWithMorphTargets=true
  bUsedWithSkeletalMesh=true
  bUsedWithStaticLighting=true
  shadingModel=MSM_FromMaterialExpression
  usedShadingModels=Default Lit | Subsurface | Clear Coat | Subsurface Profile | Cloth
 }
 nodes [
  n1 = MaterialFunctionCall(Color=NamedRerouteUsage()) {materialFunction=&SplitComponents}
  Add(A=@n1.R, B=@n1.R)
  n2 = MaterialFunctionCall() {materialFunction=&MF_CharacterParent_2023_PreFX}
  n3 = GetMaterialAttributes(None=MaterialFunctionCall!(Material=NamedRerouteUsage(), Tech Art Mask=@n2.Tech Art Mask) {materialFunction=/CosmeticCompanions/Materials/MaterialFunctions/MF_FN_Companion_Custom_Color}.Base Color) {attributeGetTypes [#1]}
  n4 = ComponentMask(None=@n2.Tech Art Mask) {g=true}
  n5 = TextureCoordinate()
  n6 = Step(X=SphereMask(A=ComponentMask(None=@n5) {r=true}, B=Constant() {r=1}) {attenuationRadius=0.9}) {constY=0.1}
  n7 = Add(A=ComponentMask(None=NamedRerouteUsage()) {r=true}) {constB=2}
  n8 = Multiply(A=OneMinus(None=Step(X=@n7)), B=SphereMask(A=NamedRerouteUsage(), B=Constant3Vector() {constant=<rgba -6 11 50 0>}) {attenuationRadius=8})
  n9 = Saturate(None=Subtract(A=Multiply(A=@n4, B=@n6), B=Multiply(A=@n8, B=@n4)))
  n10 = LinearInterpolate(
   A=@n3.Base Color
   B=VectorParameter() {parameterName=SecondaryColor_Variation, defaultValue=<rgba 0 0.2346 0.5841 1>}.RGB
   Alpha=NamedRerouteDeclaration(None=Step(X=@n9) {constY=0.1}) {name=BrowsMask, nodeColor=<rgba 1 0 0.5859 1>}
  )
  MaterialFunctionCall(
   Material=MaterialFunctionCall!(Material=SetMaterialAttributes!(MaterialAttributes=@n3.MaterialAttributes, Base Color=@n10) {attributeSetTypes [#1]}) {materialFunction=/CosmeticCompanions/Materials/MaterialFunctions/MF_FN_Companion_Flipbook}.Result
   M=@n2.M
   Pre Skinned Local Position=@n2.M
   Distance Blend=@n2.M
   Fresnel=@n2.M
  ) {materialFunction=&MF_CharacterParent_2023_PostFX}
  NamedRerouteDeclaration(None=@n2.Pre skinned Local Position) {name=PreSkinned_Position, nodeColor=<rgba 1 0.0938 0 1>}
  n11 = GetMaterialAttributes(None=NamedRerouteUsage()) {attributeGetTypes [
    #2
    #3
    #1
    #4
   ]}
  NamedRerouteDeclaration(None=@n11.MaterialAttributes) {name=Material Attributes, nodeColor=<rgba 0 0.2031 1 1>}
  NamedRerouteDeclaration(None=@n2.Result) {name=OriginalMA, nodeColor=<rgba 1 0 0.5156 1>}
  n12 = SetMaterialAttributes!(MaterialAttributes=@n11.MaterialAttributes, Fuzz Color or Subsurface Color=@n11.MaterialAttributes, Opacity=Constant() {r=0.5}) {attributeSetTypes [#2, #3]}
  n13 = Multiply(A=VectorParameter() {group=06_SS Shader Model, parameterName=SS Color, defaultValue=<rgba 1 1 1 0>}.RGB) {constB=0.6}
  n14 = Multiply(A=@n11.Base Color, B=@n13)
  n15 = Power(Base=NamedRerouteUsage(), Exp=ScalarParameter() {defaultValue=4, group=06_SS Shader Model, parameterName=SSFresnelExponent})
  n16 = StaticSwitchParameter(
   True=LinearInterpolate(
    A=ScalarParameter() {group=06_SS Shader Model, parameterName=SSInnerStrength}
    B=ScalarParameter() {defaultValue=0.85, group=06_SS Shader Model, parameterName=SSOuterStrength}
    Alpha=Saturate(None=@n15)
   )
   False=ScalarParameter() {group=06_SS Shader Model, parameterName=SSOpacity}
  ) {group=06_SS Shader Model, parameterName=UseFresnelforSSOpacity}
  n17 = SetMaterialAttributes!(
   MaterialAttributes=NamedRerouteUsage()
   Fuzz Color or Subsurface Color=Multiply(A=StaticSwitchParameter(True=@n14, False=@n13) {group=06_SS Shader Model, parameterName=UseDiffuseInSSColor}, B=@n16)
   Opacity=@n16
   Shading Model=ShadingModel() {shadingModel=MSM_Subsurface}
  ) {attributeSetTypes [#2, #3, #5]}
  n18 = Saturate(None=@n1.R)
  n19 = TextureSampleParameter2D() {
   group=Color Customization
   parameterName=Custom Color Mask
   samplerSource=SSM_Wrap_WorldGroupSettings
   samplerType=SAMPLERTYPE_Masks
   sortPriority=11
   texture=&T_BlackMask
  }
  n20 = Subtract(A=Saturate(None=Subtract(A=@n18, B=@n19.G)), B=@n19.R)
  NamedRerouteDeclaration(None=BlendMaterialAttributes(
    A=NamedRerouteUsage()
    B=MaterialFunctionCall!(
     Low _ PC Low and Med=@n12
     "High _ XB1, PS4"=@n12
     "Epic _ PC High, XSX, PS5"=@n17
     --------
     "Mobile Low _ PC Perf, Android"=@n12
     Mobile Med _ Switch=@n12
    ) {materialFunction=/Game/Global/MaterialFunctions/TechArt/MF_Fortnite_QualitySwitch_MaterialAttributes}.Result
    Alpha=Saturate(None=@n20)
   )) {name=SubsurfaceMA, nodeColor=<rgba 0 0.8359 1 1>}
  NamedRerouteDeclaration(None=@n2.Tech Art Mask) {name=TechArtMask, nodeColor=<rgba 1 0.8438 0 1>}
  NamedRerouteDeclaration(None=@n2.Fresnel) {name=InFresnel, nodeColor=<rgba 0.5234 0 1 1>}
 ]
}
