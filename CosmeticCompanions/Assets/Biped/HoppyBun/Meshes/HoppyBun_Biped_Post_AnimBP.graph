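# Post-process AnimBlueprint for the HoppyBun biped cosmetic companion, layering pose corrections and secondary motion onto the incoming pose.
# The input pose runs through a PoseAsset blend, then takes an additive sequence weighted by CosmeticPoseAlpha.
# A chain of per-bone AnimDynamics springs, sharing DynamicsAlpha, simulates both ears and two bands in root-relative space.
# Ears are stiffer and more tightly limited than the looser bands; the EventGraph is empty.
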AnimBlueprint /CosmeticCompanions/Assets/Biped/HoppyBun/Meshes/HoppyBun_Biped_Post_AnimBP
target_skeleton=&/HoppyBun_Biped_Skeleton
graphs [
 AnimGraph {nodes [
   n1 = Pose!(
    PoseAsset:Pose Asset Object Reference=Get PoseAsset() -> (PoseAsset:Pose Asset Object Reference)
    SourcePose:Pose Link Structure=Input Pose()
   ) -> (Pose:Pose Link Structure) {is=title AnimGraphNode_PoseBlendNode, bIgnorePoseAssetSourceAnimationCurves=true, blendOption=Linear}
   n2 = Sequence Player!(Sequence:Anim Sequence Base Object Reference=Property Access!() -> (Value:Anim Sequence Object Reference) {is=title K2Node_PropertyAccess}) -> (Pose)
   n3 = Get CosmeticPoseAlpha() -> (CosmeticPoseAlpha:float)
   n4 = Get DynamicsAlpha() -> (DynamicsAlpha:float)
   n5 = Anim Dynamics - Bone: dyn_L_ear_1(ComponentPose=Local To Component(LocalPose=Apply Additive(Base=@n1, Additive=@n2, Alpha=@n3)), Alpha=@n4, ExternalForce="0,0,0") {
    angularBiasOverride=1
    angularDampingOverride=0.9999
    angularSpringConstant=4000
    bAngularSpring=true
    bOverrideAngularBias=true
    bOverrideAngularDamping=true
    bUseGravityOverride=true
    simulationSpace=RootRelative
    alphaBoolBlend {blendOption=Linear}
    alphaScaleBiasClamp {bias=1, scale=-1}
    boundBone {boneName=dyn_L_ear_1}
    componentAppliedLinearAccClamp=<3000 3000 3000>
    gravityOverride {z=-750}
    physicsBodyDefinitions [{
      boundBone {boneName=dyn_L_ear_1}
      boxExtents=<15 15 15>
      constraintSetup {angularLimitsMax=<0 15 15>, angularLimitsMin=<0 -15 -15>, angularTarget=<1 0 0>}
      localJointOffset=<1 0 0>
     }]
    simSpaceSettings {maxAngularAcceleration=45, maxAngularVelocity=8, simSpaceAngularAlpha=0.25}
   }
   n6 = Anim Dynamics - Bone: dyn_L_ear_2(ComponentPose=@n5, Alpha=@n4, ExternalForce="0,0,0") {
    angularBiasOverride=1
    angularDampingOverride=1
    angularSpringConstant=4000
    bAngularSpring=true
    bOverrideAngularBias=true
    bOverrideAngularDamping=true
    bUseGravityOverride=true
    simulationSpace=RootRelative
    alphaBoolBlend {blendOption=Linear}
    alphaScaleBiasClamp {bias=1, scale=-1}
    boundBone {boneName=dyn_L_ear_2}
    componentAppliedLinearAccClamp=<3000 3000 3000>
    gravityOverride {z=-750}
    physicsBodyDefinitions [{
      boundBone {boneName=dyn_L_ear_2}
      boxExtents=<15 15 15>
      constraintSetup {angularLimitsMax=<0 5 5>, angularLimitsMin=<0 -5 -5>, angularTarget=<1 0 0>}
      localJointOffset=<0.25 0 0>
     }]
    simSpaceSettings {maxAngularAcceleration=45, maxAngularVelocity=8, simSpaceAngularAlpha=0.25}
   }
   n7 = Anim Dynamics - Bone: dyn_L_ear_3(ComponentPose=@n6, Alpha=@n4, ExternalForce="0,0,0") {
    angularBiasOverride=1
    angularDampingOverride=1
    angularSpringConstant=4000
    bAngularSpring=true
    bOverrideAngularBias=true
    bOverrideAngularDamping=true
    bUseGravityOverride=true
    simulationSpace=RootRelative
    alphaBoolBlend {blendOption=Linear}
    alphaScaleBiasClamp {bias=1, scale=-1}
    boundBone {boneName=dyn_L_ear_3}
    componentAppliedLinearAccClamp=<3000 3000 3000>
    gravityOverride {z=-750}
    physicsBodyDefinitions [{
      boundBone {boneName=dyn_L_ear_3}
      boxExtents=<15 15 15>
      constraintSetup {angularLimitsMax=<0 5 5>, angularLimitsMin=<0 -5 -5>, angularTarget=<1 0 0>}
      localJointOffset=<0.25 0 0>
     }]
    simSpaceSettings {maxAngularAcceleration=45, maxAngularVelocity=8, simSpaceAngularAlpha=0.25}
   }
   n8 = Anim Dynamics - Bone: dyn_R_ear_1(ComponentPose=@n7, Alpha=@n4, ExternalForce="0,0,0") {
    angularBiasOverride=1
    angularDampingOverride=0.9999
    angularSpringConstant=4000
    bAngularSpring=true
    bOverrideAngularBias=true
    bOverrideAngularDamping=true
    bUseGravityOverride=true
    simulationSpace=RootRelative
    alphaBoolBlend {blendOption=Linear}
    alphaScaleBiasClamp {bias=1, scale=-1}
    boundBone {boneName=dyn_R_ear_1}
    componentAppliedLinearAccClamp=<3000 3000 3000>
    gravityOverride {z=-750}
    physicsBodyDefinitions [{
      boundBone {boneName=dyn_R_ear_1}
      boxExtents=<15 15 15>
      constraintSetup {angularLimitsMax=<0 15 15>, angularLimitsMin=<0 -15 -15>, angularTarget=<1 0 0>}
      localJointOffset=<1 0 0>
     }]
    simSpaceSettings {maxAngularAcceleration=45, maxAngularVelocity=8, simSpaceAngularAlpha=0.25}
   }
   n9 = Anim Dynamics - Bone: dyn_R_ear_2(ComponentPose=@n8, Alpha=@n4, ExternalForce="0,0,0") {
    angularBiasOverride=1
    angularDampingOverride=1
    angularSpringConstant=4000
    bAngularSpring=true
    bOverrideAngularBias=true
    bOverrideAngularDamping=true
    bUseGravityOverride=true
    simulationSpace=RootRelative
    alphaBoolBlend {blendOption=Linear}
    alphaScaleBiasClamp {bias=1, scale=-1}
    boundBone {boneName=dyn_R_ear_2}
    componentAppliedLinearAccClamp=<3000 3000 3000>
    gravityOverride {z=-750}
    physicsBodyDefinitions [{
      boundBone {boneName=dyn_R_ear_2}
      boxExtents=<15 15 15>
      constraintSetup {angularLimitsMax=<0 5 5>, angularLimitsMin=<0 -5 -5>, angularTarget=<1 0 0>}
      localJointOffset=<0.25 0 0>
     }]
    simSpaceSettings {maxAngularAcceleration=45, maxAngularVelocity=8, simSpaceAngularAlpha=0.25}
   }
   n10 = Anim Dynamics - Bone: dyn_R_ear_3(ComponentPose=@n9, Alpha=@n4, ExternalForce="0,0,0") {
    angularBiasOverride=1
    angularDampingOverride=1
    angularSpringConstant=4000
    bAngularSpring=true
    bOverrideAngularBias=true
    bOverrideAngularDamping=true
    bUseGravityOverride=true
    simulationSpace=RootRelative
    alphaBoolBlend {blendOption=Linear}
    alphaScaleBiasClamp {bias=1, scale=-1}
    boundBone {boneName=dyn_R_ear_3}
    componentAppliedLinearAccClamp=<3000 3000 3000>
    gravityOverride {z=-750}
    physicsBodyDefinitions [{
      boundBone {boneName=dyn_R_ear_3}
      boxExtents=<15 15 15>
      constraintSetup {angularLimitsMax=<0 5 5>, angularLimitsMin=<0 -5 -5>, angularTarget=<1 0 0>}
      localJointOffset=<0.25 0 0>
     }]
    simSpaceSettings {maxAngularAcceleration=45, maxAngularVelocity=8, simSpaceAngularAlpha=0.25}
   }
   n11 = Anim Dynamics - Bone: dyn_band_1_A(ComponentPose=@n10, Alpha=@n4, ExternalForce="0,0,0") {
    angularBiasOverride=1
    angularDampingOverride=0.9999
    angularSpringConstant=2000
    bAngularSpring=true
    bOverrideAngularBias=true
    bOverrideAngularDamping=true
    bUseGravityOverride=true
    simulationSpace=RootRelative
    alphaBoolBlend {blendOption=Linear}
    alphaScaleBiasClamp {bias=1, scale=-1}
    boundBone {boneName=dyn_band_1_A}
    componentAppliedLinearAccClamp=<3000 3000 3000>
    gravityOverride {z=-490}
    physicsBodyDefinitions [{
      boundBone {boneName=dyn_band_1_A}
      boxExtents=<15 15 15>
      constraintSetup {angularLimitsMax=<0 25 45>, angularLimitsMin=<0 -25 -45>, angularTarget=<1 0 0>}
      localJointOffset=<3 0 0>
     }]
    simSpaceSettings {maxAngularAcceleration=45, maxAngularVelocity=8, simSpaceAngularAlpha=1}
   }
   n12 = Anim Dynamics - Bone: dyn_band_1_B(ComponentPose=@n11, Alpha=@n4, ExternalForce="0,0,0") {
    angularBiasOverride=1
    angularDampingOverride=0.9999
    angularSpringConstant=2000
    bAngularSpring=true
    bOverrideAngularBias=true
    bOverrideAngularDamping=true
    bUseGravityOverride=true
    simulationSpace=RootRelative
    alphaBoolBlend {blendOption=Linear}
    alphaScaleBiasClamp {bias=1, scale=-1}
    boundBone {boneName=dyn_band_1_B}
    componentAppliedLinearAccClamp=<3000 3000 3000>
    gravityOverride {z=-490}
    physicsBodyDefinitions [{
      boundBone {boneName=dyn_band_1_B}
      boxExtents=<15 15 15>
      constraintSetup {angularLimitsMax=<0 25 45>, angularLimitsMin=<0 -25 -45>, angularTarget=<1 0 0>}
      localJointOffset=<3 0 0>
     }]
    simSpaceSettings {maxAngularAcceleration=45, maxAngularVelocity=8, simSpaceAngularAlpha=1}
   }
   n13 = Anim Dynamics - Bone: dyn_band_1_C(ComponentPose=@n12, Alpha=@n4, ExternalForce="0,0,0") {
    angularBiasOverride=1
    angularDampingOverride=0.9999
    angularSpringConstant=2000
    bAngularSpring=true
    bOverrideAngularBias=true
    bOverrideAngularDamping=true
    bUseGravityOverride=true
    simulationSpace=RootRelative
    alphaBoolBlend {blendOption=Linear}
    alphaScaleBiasClamp {bias=1, scale=-1}
    boundBone {boneName=dyn_band_1_C}
    componentAppliedLinearAccClamp=<3000 3000 3000>
    gravityOverride {z=-490}
    physicsBodyDefinitions [{
      boundBone {boneName=dyn_band_1_C}
      boxExtents=<15 15 15>
      constraintSetup {angularLimitsMax=<0 25 45>, angularLimitsMin=<0 -25 -45>, angularTarget=<1 0 0>}
      localJointOffset=<3 0 0>
     }]
    simSpaceSettings {maxAngularAcceleration=45, maxAngularVelocity=8, simSpaceAngularAlpha=1}
   }
   n14 = Anim Dynamics - Bone: dyn_band_2_A(ComponentPose=@n13, Alpha=@n4, ExternalForce="0,0,0") {
    angularBiasOverride=1
    angularDampingOverride=0.9999
    angularSpringConstant=2000
    bAngularSpring=true
    bOverrideAngularBias=true
    bOverrideAngularDamping=true
    bUseGravityOverride=true
    simulationSpace=RootRelative
    alphaBoolBlend {blendOption=Linear}
    alphaScaleBiasClamp {bias=1, scale=-1}
    boundBone {boneName=dyn_band_2_A}
    componentAppliedLinearAccClamp=<3000 3000 3000>
    gravityOverride {z=-490}
    physicsBodyDefinitions [{
      boundBone {boneName=dyn_band_2_A}
      boxExtents=<15 15 15>
      constraintSetup {angularLimitsMax=<0 25 45>, angularLimitsMin=<0 -25 -45>, angularTarget=<1 0 0>}
      localJointOffset=<3 0 0>
     }]
    simSpaceSettings {maxAngularAcceleration=45, maxAngularVelocity=8, simSpaceAngularAlpha=1}
   }
   n15 = Anim Dynamics - Bone: dyn_band_2_B(ComponentPose=@n14, Alpha=@n4, ExternalForce="0,0,0") {
    angularBiasOverride=1
    angularDampingOverride=0.9999
    angularSpringConstant=2000
    bAngularSpring=true
    bOverrideAngularBias=true
    bOverrideAngularDamping=true
    bUseGravityOverride=true
    simulationSpace=RootRelative
    alphaBoolBlend {blendOption=Linear}
    alphaScaleBiasClamp {bias=1, scale=-1}
    boundBone {boneName=dyn_band_2_B}
    componentAppliedLinearAccClamp=<3000 3000 3000>
    gravityOverride {z=-490}
    physicsBodyDefinitions [{
      boundBone {boneName=dyn_band_2_B}
      boxExtents=<15 15 15>
      constraintSetup {angularLimitsMax=<0 25 45>, angularLimitsMin=<0 -25 -45>, angularTarget=<1 0 0>}
      localJointOffset=<3 0 0>
     }]
    simSpaceSettings {maxAngularAcceleration=45, maxAngularVelocity=8, simSpaceAngularAlpha=1}
   }
   n16 = Anim Dynamics - Bone: dyn_band_2_C(ComponentPose=@n15, Alpha=@n4, ExternalForce="0,0,0") {
    angularBiasOverride=1
    angularDampingOverride=0.9999
    angularSpringConstant=2000
    bAngularSpring=true
    bOverrideAngularBias=true
    bOverrideAngularDamping=true
    bUseGravityOverride=true
    simulationSpace=RootRelative
    alphaBoolBlend {blendOption=Linear}
    alphaScaleBiasClamp {bias=1, scale=-1}
    boundBone {boneName=dyn_band_2_C}
    componentAppliedLinearAccClamp=<3000 3000 3000>
    gravityOverride {z=-490}
    physicsBodyDefinitions [{
      boundBone {boneName=dyn_band_2_C}
      boxExtents=<15 15 15>
      constraintSetup {angularLimitsMax=<0 25 45>, angularLimitsMin=<0 -25 -45>, angularTarget=<1 0 0>}
      localJointOffset=<3 0 0>
     }]
    simSpaceSettings {maxAngularAcceleration=45, maxAngularVelocity=8, simSpaceAngularAlpha=1}
   }
   Output Pose(Result=Component To Local(ComponentPose=@n16))
  ]}
 EventGraph {links [], nodes []}
]
variables [PoseAsset {type=PoseAsset, default=&/HoppyBun_Biped_PoseAsset}]
